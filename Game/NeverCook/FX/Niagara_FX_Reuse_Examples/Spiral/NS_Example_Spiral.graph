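# Niagara example system for a spiral effect, kept in a NeverCook folder of FX reuse examples.
# It holds two emitters: SwirlLooping, and SwirlWithErodeOut, which is disabled.
# SystemState is the only listed module, and the system exposes no user variables and lists no renderers for either emitter.
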NiagaraSystem /Game/NeverCook/FX/Niagara_FX_Reuse_Examples/Spiral/NS_Example_Spiral
dependencies {usedDataInterfaces [], usedDynamicInputs [], usedModules [&SystemState]}
emitters {SwirlLooping {rendererClasses [], renderers []}, SwirlWithErodeOut {bEnabled=false, rendererClasses [], renderers []}}
summary {
 systemName=NS_Example_Spiral
 emitters [{emitterName=SwirlLooping, rendererClasses []}, {bEnabled=false, emitterName=SwirlWithErodeOut, rendererClasses []}]
 userVariables []
}
